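# M_UI_Headset is a UI-domain material in the ItemShop headset art folder.
# It contains a single texture parameter, SeriesTexture, which defaults to the T_UI_Headset texture from OfferCatalog so material instances can override the image.
# The graph shows no connection from that sampler to a material output and no other processing.
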Material /ItemShop/UI/Art/Headset/M_UI_Headset
graph {
 properties {bUsedWithStaticMesh=false, materialDomain=MD_UI, pixelDepthOffsetMode=PDOM_Legacy, usedShadingModels=Default Lit}
 nodes [TextureSampleParameter2D() {parameterName=SeriesTexture, texture=/OfferCatalog/Art/Headset/T_UI_Headset}]
}
