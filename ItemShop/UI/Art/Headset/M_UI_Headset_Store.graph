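# UI-domain material for the Item Shop that draws a headset graphic glowing red over a dark background.
# A rotated linear gradient from charcoal to deep red forms the backdrop.
# Two headset textures share fitted, scaled and offset UVs, are tinted by a squared-sine time pulse between red and black, then summed.
# A radial gradient fades that graphic before it is blended with the backdrop.
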Material /ItemShop/UI/Art/Headset/M_UI_Headset_Store
graph {
 properties {bUsedWithStaticMesh=false, materialDomain=MD_UI, pixelDepthOffsetMode=PDOM_Legacy, usedShadingModels=Default Lit}
 nodes [
  n1 = ScalarParameter() {parameterName=Gstart}
  n2 = ScalarParameter() {defaultValue=0.9521, parameterName=GEnd}
  n3 = ScalarParameter() {defaultValue=0.75, parameterName=GAngle}
  n4 = Multiply(A=Time(), B=ScalarParameter() {defaultValue=0.019, parameterName=tenminutes})
  n5 = LinearInterpolate(
   A=Multiply(A=ScalarParameter() {defaultValue=0.68, parameterName=Opacity}, B=Constant3Vector() {constant=<rgba 0.8632 0 0.0802 0>})
   B=Constant3Vector()
   Alpha=Multiply(A=Sine(None=@n4), B=Sine(None=@n4))
  )
  n6 = Add(
   A=MaterialFunctionCall(
    UVs=MaterialFunctionCall() {materialFunction=&MF_UI_UVScaling}.Scaled to Fit
    Texture Scale=ScalarParameter() {defaultValue=2.25, parameterName=Scale}
   ) {materialFunction=&ScaleUVsByCenter}.UVs
   B=MaterialFunctionCall(X=ScalarParameter() {defaultValue=0.005, parameterName=X}, Y=ScalarParameter() {defaultValue=0.088, parameterName=Y}) {materialFunction=&MakeFloat2}.Result
  )
  n7 = LinearInterpolate(B=@n5, Alpha=TextureSample(UVs=@n6) {texture=&/T_UI_HeadsetBGMask}.R)
  n8 = LinearInterpolate(B=@n5, Alpha=TextureSample(UVs=@n6) {texture=&/T_UI_HeadsetBG}.RGB)
  n9 = ScalarParameter() {defaultValue=-0.12, parameterName=GradientStart}
  n10 = ScalarParameter() {defaultValue=0.9386, parameterName=GradientEnd}
  LinearInterpolate(
   A=LinearInterpolate(
    A=Constant3Vector() {constant=<rgba 0.026 0.026 0.026 0>}
    B=Constant3Vector() {constant=<rgba 0.0521 0 0.0048 0>}
    Alpha=MaterialFunctionCall(GradientStart=@n1, GradientEnd=@n2, RotationAngle=@n3) {materialFunction=&MF_UI_LinearGradient}.Result
   )
   B=LinearInterpolate(
    B=Add(A=@n7, B=@n8)
    Alpha=MaterialFunctionCall(UV=@n6, GradientStart=@n9, GradientEnd=@n10) {materialFunction=&MF_UI_RadialGradient}.Smooth
   )
  )
  MaterialFunctionCall() {materialFunction=&MF_UI_UVScaling}
  Multiply()
 ]
}
